# Copyright Epic Games, Inc. All Rights Reserved.
#################################################
# Generated Digest of Verse API
# DO NOT modify this manually!
# Generated from build: ++Fortnite+Release-36.20-CL-43847582
#################################################


using {/Verse.org/Native}
using {/Verse.org/SceneGraph}
using {/Verse.org/Simulation}
using {/Verse.org/Colors}
using {/Verse.org/Assets}
# Module import path: /playkigai.dig@fortnite.com/Minigame_LaserTennis/Materials
Materials := module:
    M_Laser<scoped {/playkigai.dig@fortnite.com/Minigame_LaserTennis}> := class<final><scoped {/playkigai.dig@fortnite.com/Minigame_LaserTennis}>(material):
        @editable
        var Color<public>:color = external {}

    M_PressurePlate<scoped {/playkigai.dig@fortnite.com/Minigame_LaserTennis}> := class<final><scoped {/playkigai.dig@fortnite.com/Minigame_LaserTennis}>(material):
        @editable
        var BaseColor<public>:color = external {}

    MI_Laser_Blue<scoped {/playkigai.dig@fortnite.com/Minigame_LaserTennis}> := class<final><scoped {/playkigai.dig@fortnite.com/Minigame_LaserTennis}>(material):
        @editable
        var Color<public>:color = external {}

    MI_Laser_Red<scoped {/playkigai.dig@fortnite.com/Minigame_LaserTennis}> := class<final><scoped {/playkigai.dig@fortnite.com/Minigame_LaserTennis}>(material):
        @editable
        var Color<public>:color = external {}

    MI_PressurePlate_Charged<scoped {/playkigai.dig@fortnite.com/Minigame_LaserTennis}> := class<final><scoped {/playkigai.dig@fortnite.com/Minigame_LaserTennis}>(material):
        @editable
        var BaseColor<public>:color = external {}

    MI_PressurePlate_Uncharged<scoped {/playkigai.dig@fortnite.com/Minigame_LaserTennis}> := class<final><scoped {/playkigai.dig@fortnite.com/Minigame_LaserTennis}>(material):
        @editable
        var BaseColor<public>:color = external {}

# Module import path: /playkigai.dig@fortnite.com/Minigame_LaserTennis/Meshes
Meshes := module:
    SM_CharacterCollision<scoped {/playkigai.dig@fortnite.com/Minigame_LaserTennis}> := class<final><scoped {/playkigai.dig@fortnite.com/Minigame_LaserTennis}>(mesh_component):
        @editable
        var Material_0<public>:material = external {}

    SM_CharacterCollision_asset<scoped {/playkigai.dig@fortnite.com/Minigame_LaserTennis}>:mesh = external {}

    SM_Laser<scoped {/playkigai.dig@fortnite.com/Minigame_LaserTennis}> := class<final><scoped {/playkigai.dig@fortnite.com/Minigame_LaserTennis}>(mesh_component):
        @editable
        var Material_0<public>:material = external {}

    SM_Laser_asset<scoped {/playkigai.dig@fortnite.com/Minigame_LaserTennis}>:mesh = external {}

    SM_LaserFieldFloor<scoped {/playkigai.dig@fortnite.com/Minigame_LaserTennis}> := class<final><scoped {/playkigai.dig@fortnite.com/Minigame_LaserTennis}>(mesh_component):
        @editable
        var Material_0<public>:material = external {}

    SM_LaserFieldFloor_asset<scoped {/playkigai.dig@fortnite.com/Minigame_LaserTennis}>:mesh = external {}

    SM_NewLaserSpawner<scoped {/playkigai.dig@fortnite.com/Minigame_LaserTennis}> := class<final><scoped {/playkigai.dig@fortnite.com/Minigame_LaserTennis}>(mesh_component):
        @editable
        var Material_0<public>:material = external {}

    SM_NewLaserSpawner_asset<scoped {/playkigai.dig@fortnite.com/Minigame_LaserTennis}>:mesh = external {}

    SM_PressurePlate<scoped {/playkigai.dig@fortnite.com/Minigame_LaserTennis}> := class<final><scoped {/playkigai.dig@fortnite.com/Minigame_LaserTennis}>(mesh_component):
        @editable
        var (/playkigai.dig@fortnite.com/Minigame_LaserTennis/Meshes/SM_PressurePlate:)M_PressurePlate<public>:material = external {}

    SM_PressurePlate_asset<scoped {/playkigai.dig@fortnite.com/Minigame_LaserTennis}>:mesh = external {}

# Module import path: /playkigai.dig@fortnite.com/Minigame_LaserTennis/Prefabs
Prefabs := module:
    @import_as("/Minigame_LaserTennis/Prefabs/character_collision_proxy.character_collision_proxy_C")
    character_collision_proxy<scoped {/playkigai.dig@fortnite.com/Minigame_LaserTennis}> := class<final>(entity):


    character_collision_proxy_asset<scoped {/playkigai.dig@fortnite.com/Minigame_LaserTennis}>:entity_prefab = external {}

    @import_as("/Minigame_LaserTennis/Prefabs/collision_proxy_spawner.collision_proxy_spawner_C")
    collision_proxy_spawner<scoped {/playkigai.dig@fortnite.com/Minigame_LaserTennis}> := class<final>(entity):


    collision_proxy_spawner_asset<scoped {/playkigai.dig@fortnite.com/Minigame_LaserTennis}>:entity_prefab = external {}

    @import_as("/Minigame_LaserTennis/Prefabs/game_manager.game_manager_C")
    game_manager<scoped {/playkigai.dig@fortnite.com/Minigame_LaserTennis}> := class<final>(entity):


    game_manager_asset<scoped {/playkigai.dig@fortnite.com/Minigame_LaserTennis}>:entity_prefab = external {}

    @import_as("/Minigame_LaserTennis/Prefabs/laser.laser_C")
    laser<scoped {/playkigai.dig@fortnite.com/Minigame_LaserTennis}> := class<final>(entity):


    laser_asset<scoped {/playkigai.dig@fortnite.com/Minigame_LaserTennis}>:entity_prefab = external {}

    @import_as("/Minigame_LaserTennis/Prefabs/laser_board.laser_board_C")
    laser_board<scoped {/playkigai.dig@fortnite.com/Minigame_LaserTennis}> := class<final>(entity):


    laser_board_asset<scoped {/playkigai.dig@fortnite.com/Minigame_LaserTennis}>:entity_prefab = external {}

    @import_as("/Minigame_LaserTennis/Prefabs/laser_field.laser_field_C")
    laser_field<scoped {/playkigai.dig@fortnite.com/Minigame_LaserTennis}> := class<final>(entity):


    laser_field_asset<scoped {/playkigai.dig@fortnite.com/Minigame_LaserTennis}>:entity_prefab = external {}

    @import_as("/Minigame_LaserTennis/Prefabs/laser_spawner.laser_spawner_C")
    laser_spawner<scoped {/playkigai.dig@fortnite.com/Minigame_LaserTennis}> := class<final>(entity):


    laser_spawner_asset<scoped {/playkigai.dig@fortnite.com/Minigame_LaserTennis}>:entity_prefab = external {}

    @import_as("/Minigame_LaserTennis/Prefabs/pressure_plate.pressure_plate_C")
    pressure_plate<scoped {/playkigai.dig@fortnite.com/Minigame_LaserTennis}> := class<final>(entity):


    pressure_plate_asset<scoped {/playkigai.dig@fortnite.com/Minigame_LaserTennis}>:entity_prefab = external {}